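# Unlit, additive, two-sided flame material for the Galaxy Level pickaxe effects, usable on Niagara mesh particles and ribbons.
# Panning plasma noise on bent, vertex-color-offset UVs combines with fire-line and breakup masks into a contrasted flame shape.
# The shape drives time-of-day-scaled HDR color and opacity with depth, edge and view-angle fades.
# Vertex-red-masked offset adds noise jiggle and Z stretch, while dynamic parameters control ignition, motion, complexity and intensity.
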Material /BRCosmetics/Weapons/FORT_Melee/Pickaxe_GalaxyLevel/FX/M_GalaxyLevelPickaxe_Flamesv2
graph {
 properties {
  bUsedWithNiagaraMeshParticles=true
  bUsedWithNiagaraRibbons=true
  blendMode=BLEND_Additive
  pixelDepthOffsetMode=PDOM_Legacy
  shadingModel=MSM_Unlit
  twoSided=true
  usedShadingModels=Unlit
 }
 nodes [
  n1 = VectorParameter() {parameterName=TimeofDayScale, defaultValue=<rgba 0.333 0.5 0.4 0.333>}
  n2 = LinearInterpolate(
   A=VectorParameter() {group=FlameShape, parameterName=Color_A, defaultValue=<rgba 0.5 0 0.4583 1>}.RGB
   B=VectorParameter() {group=FlameShape, parameterName=Color_B, defaultValue=<rgba 1 0 0.9167 1>}.RGB
   Alpha=Power(Base=NamedRerouteUsage()) {constExponent=0.5}
  )
  n3 = ParticleColor()
  n4 = LinearInterpolate(A=Multiply(A=@n3.RGB) {constB=0.25}, B=@n3.RGB, Alpha=NamedRerouteUsage())
  n5 = ScalarParameter() {parameterName=Use Particle Color}
  n6 = Multiply(
   A=Multiply(
    A=MaterialFunctionCall(Value_Morning=@n1.R, Value_Day=@n1.R, Value_Evening=@n1.R, Value_Night=@n1.R) {materialFunction=&Scalar_Time_of_Day}.Result
    B=LinearInterpolate(A=@n2, B=@n4, Alpha=@n5)
   )
   B=LinearInterpolate(
    A=VectorParameter() {group=FlameShape, parameterName=HDR_Tint, defaultValue=<rgba 1 0 0.9167 1>}.RGB
    B=ParticleColor().RGB
    Alpha=ScalarParameter() {parameterName=Use Particle Color}
   )
  )
  n7 = DynamicParameter() {defaultValue=<rgba 1 1 2 1>, paramNames [Fire UV Motion, Enhance Flame, Add Flame Complexity, HDR Intensity]}
  MaterialFunctionCall(Emissive=MaterialFunctionCall(LDR Input=Saturate(None=@n6), HDR Tint=@n6, HDR Intensity Multiplier=@n7.HDR Intensity) {materialFunction=&DeriveHDRfromLDR}.RemappedRange) {materialFunction=&MF_FX_EmissiveAdjustments}
  MaterialFunctionCall() {materialFunction=&DebugTimeSine}
  n8 = NamedRerouteUsage()
  n9 = Sine(None=ComponentMask(None=@n8) {r=true}) {period=2}
  n10 = Power(Base=ComponentMask(None=TextureCoordinate()) {g=true}, Exp=LinearInterpolate(Alpha=OneMinus(None=@n9)) {constA=2, constB=4})
  Multiply(A=@n10) {constB=2}
  n11 = NamedRerouteUsage()
  n12 = NamedRerouteUsage()
  n13 = DynamicParameter() {defaultValue=<rgba 1 1 2 1>, paramNames [Fire UV Motion, Enhance Flame, Add Flame Complexity, HDR Intensity]}
  n14 = Multiply(A=NamedRerouteUsage(), B=Saturate(None=Multiply(A=@n12, B=@n13.Enhance Flame)))
  n15 = DepthFade(Opacity=@n14, FadeDistance=ScalarParameter() {defaultValue=5, group=Alpha, parameterName=Fade Distance})
  n16 = TextureCoordinate()
  n17 = Power(Base=Sine(None=ComponentMask(None=@n16) {r=true}) {period=2}, Exp=ScalarParameter() {defaultValue=2, parameterName=Edge Fade Contrast})
  n18 = ScalarParameter() {defaultValue=10, parameterName=Edge Fade Intensity}
  n19 = Multiply(A=Multiply(A=ShadingPathSwitch(Default=@n15, Mobile=@n14), B=ParticleColor().A), B=Saturate(None=Multiply(A=@n17, B=@n18)))
  n20 = ScalarParameter() {defaultValue=1, parameterName=Edge Fade Sensitivity}
  n21 = ComponentMask(None=TextureCoordinate()) {g=true}
  Multiply(
   A=Multiply(
    A=Add(A=@n11, B=@n19)
    B=MaterialFunctionCall("FadeSensitivity (optional)"=@n20) {materialFunction=&FadeBasedOnViewAngle}.Result
   )
   B=LinearInterpolate(B=OneMinus(None=@n21), Alpha=ScalarParameter() {parameterName=Fade Bottom Edge}) {constA=1}
  )
  n22 = VertexNormalWS()
  n23 = TextureSample(UVs=Panner(Coordinate=NamedRerouteUsage(), Time=NamedRerouteUsage()) {speedX=0.15, speedY=1}) {texture=/Game/Effects/Fort_Effects/Emotes/UltraLight/Texture/T_SoftNoise_BC7}
  n24 = Multiply(A=LinearInterpolate(Alpha=Multiply(A=@n23.R) {constB=2}) {constA=-1}, B=ScalarParameter() {defaultValue=3, parameterName=WPO Jiggle Amount})
  n25 = DynamicParameter() {parameterIndex=1, defaultValue=<rgba 0 1 0 1>, paramNames [Ignition, Z Stretch, Flaem Time Offset, Param4]}
  n26 = Transform(None=Constant3Vector() {constant=<rgba 0 0 1 0>}) {transformSourceType=TRANSFORMSOURCE_Local}
  n27 = VectorParameter() {parameterName=Velocity Stretch}
  n28 = Multiply(A=Multiply(A=Add(A=@n26, B=@n27.RGB), B=@n25.Z Stretch), B=ComponentMask(None=TextureCoordinate()) {r=true})
  n29 = ScalarParameter() {defaultValue=1, parameterName=Allow Z Stretch}
  Multiply(A=VertexColor().R, B=Add(A=Multiply(A=@n22, B=@n24), B=Multiply(A=@n28, B=@n29)))
  n30 = DynamicParameter() {defaultValue=<rgba 1 1 2 1>, paramNames [Fire UV Motion, Enhance Flame, Add Flame Complexity, HDR Intensity]}
  n31 = Add(A=VertexColor().G, B=LinearInterpolate(Alpha=@n30.Fire UV Motion) {constA=1, constB=1.2})
  n32 = VectorParameter() {group=FlameShape, parameterName=Flames_UV, defaultValue=<rgba 2 0.5 0.333 0.137>}
  n33 = AppendVector(A=@n32.B, B=@n32.B)
  n34 = Multiply(A=Add(A=TextureCoordinate(), B=Multiply(A=@n31, B=@n33)), B=Multiply(A=@n31, B=ComponentMask(None=@n32.RGB) {g=true, r=true}))
  n35 = ComponentMask(None=@n34) {r=true}
  n36 = ComponentMask(None=TextureCoordinate()) {g=true}
  n37 = Multiply(A=Power(Base=OneMinus(None=@n36), Exp=ScalarParameter() {defaultValue=4, parameterName=Bend Contrast})) {constB=2}
  n38 = LinearInterpolate(
   A=LinearInterpolate(
    A=ScalarParameter() {defaultValue=0.95, parameterName=Bend U Scale Bottom Low}
    B=ScalarParameter() {defaultValue=1.05, parameterName=Bend U Scale Bottom High}
    Alpha=@n30.Fire UV Motion
   )
   B=LinearInterpolate(
    A=ScalarParameter() {defaultValue=1, parameterName=Bend U Scale Top Low}
    B=ScalarParameter() {defaultValue=1.2, parameterName=Bend U Scale Top High}
    Alpha=@n30.Fire UV Motion
   )
   Alpha=@n37
  )
  n39 = ComponentMask(None=@n34) {g=true}
  n40 = LinearInterpolate(
   A=ScalarParameter() {defaultValue=0.75, parameterName=Bend V Scale Bottom}
   B=ScalarParameter() {defaultValue=1, parameterName=Bend V Scale Top}
   Alpha=@n37
  )
  NamedRerouteDeclaration(None=AppendVector(A=Multiply(A=@n35, B=@n38), B=Multiply(A=@n39, B=@n40))) {name=GreenVertColorOffset, nodeColor=<rgba 0 1 0.5313 1>}
  n41 = Time()
  n42 = ScalarParameter() {defaultValue=1, group=FlameShape, parameterName=Flame_Speed}
  n43 = DynamicParameter() {parameterIndex=1, defaultValue=<rgba 0 1 0 1>, paramNames [Ignition, Z Stretch, Flaem Time Offset, Param4]}
  NamedRerouteDeclaration(None=Add(A=Multiply(A=@n41, B=@n42), B=@n43.Flaem Time Offset)) {name=Flame Speed, nodeColor=<rgba 1 0.8438 0 1>}
  n44 = Saturate(None=@n10)
  n45 = VectorParameter() {parameterName=Noise BreakUp UV, defaultValue=<rgba 0.25 1 0 0>}
  n46 = Panner(Coordinate=Multiply(A=@n8, B=ComponentMask(None=@n45.RGB) {g=true, r=true})) {speedY=1}
  n47 = ComponentMask(None=TextureCoordinate()) {g=true}
  n48 = Add(
   A=@n44
   B=LinearInterpolate(B=TextureSample(UVs=@n46) {texture=/Game/Effects/Fort_Effects/Textures/Plasma/T_Plasma_01_M}.B, Alpha=OneMinus(None=@n47)) {constA=1}
  )
  n49 = TextureCoordinate()
  n50 = DynamicParameter() {parameterIndex=1, defaultValue=<rgba 0 1 0 1>, paramNames [Ignition, Z Stretch, Flaem Time Offset, Param4]}
  n51 = Power(Base=OneMinus(None=ComponentMask(None=@n49) {r=true}), Exp=LinearInterpolate(Alpha=@n50.Ignition) {constA=1, constB=20})
  NamedRerouteDeclaration(None=Multiply(A=Multiply(A=@n44, B=@n48), B=Saturate(None=@n51))) {name=BreakUp, nodeColor=<rgba 0 1 0.2266 1>}
  n52 = MaterialFunctionCall() {materialFunction=&LinearGradient}
  n53 = ScalarParameter() {defaultValue=25, group=FlameShape, parameterName=FireLine}
  NamedRerouteDeclaration(None=Multiply(A=Power(Base=@n52.VGradient, Exp=@n53), B=ScalarParameter() {defaultValue=1, parameterName=Fireline Intensity})) {name=Fire Line, nodeColor=<rgba 1 0 0.3984 1>}
  n54 = NamedRerouteUsage()
  n55 = ComponentMask(None=VectorParameter() {parameterName=Contrast Flame UV Scale, defaultValue=<rgba 1 1 0 0>}.RGB) {g=true, r=true}
  n56 = TextureSample(UVs=Panner(Coordinate=Multiply(A=@n54, B=@n55), Time=NamedRerouteUsage()) {speedX=0.25, speedY=0.3}) {texture=&T_Plasma_01_M}
  n57 = Constant() {r=0.2}
  n58 = Panner(Coordinate=NamedRerouteUsage(), Time=NamedRerouteUsage()) {speedX=0.333, speedY=1.25}
  n59 = TextureSample(UVs=Panner(Coordinate=NamedRerouteUsage(), Time=NamedRerouteUsage()) {speedX=0.26, speedY=1}) {texture=&T_Plasma_01_M}
  n60 = Multiply(
   A=Add(
    A=MaterialFunctionCall(In=@n56.G, Contrast=@n57) {materialFunction=&CheapContrast}.Result
    B=TextureSample(UVs=@n58) {texture=&T_Plasma_01_M}.R
   )
   B=@n59.R
  )
  n61 = DynamicParameter() {defaultValue=<rgba 1 1 2 1>, paramNames [Fire UV Motion, Enhance Flame, Add Flame Complexity, HDR Intensity]}
  n62 = Multiply(A=@n59.B, B=@n61.Add Flame Complexity)
  n63 = Power(Base=Add(A=Add(A=@n60, B=@n62), B=NamedRerouteUsage()), Exp=ScalarParameter() {defaultValue=0.75, parameterName=Fire Contrast}) {constExponent=0.5}
  n64 = ScalarParameter() {defaultValue=0.033, group=FlameShape, parameterName=HardEdgeness}
  NamedRerouteDeclaration(None=Saturate(None=MaterialFunctionCall(In=@n63, Contrast=@n64) {materialFunction=&CheapContrast}.Result)) {name=Flame Setup, nodeColor=<rgba 1 0.3516 0 1>}
  n65 = OneMinus(None=ComponentMask(None=TextureCoordinate()) {g=true})
  n66 = ScalarParameter() {defaultValue=1, parameterName=Alpha Addition Contrast}
  NamedRerouteDeclaration(None=LinearInterpolate(A=ScalarParameter() {defaultValue=0.01, parameterName=Alpha Addition}, Alpha=Power(Base=@n65, Exp=@n66)) {constB=0}) {name=Alpha Addition, nodeColor=<rgba 0 0.1328 1 1>}
 ]
}
