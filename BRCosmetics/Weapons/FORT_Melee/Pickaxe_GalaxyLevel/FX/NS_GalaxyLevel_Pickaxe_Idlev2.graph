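# Idle Niagara effect for the GalaxyLevel pickaxe: a magenta flame mesh with rising embers.
# FlameMesh holds a persistent mesh particle with waveform-animated material parameters and a component renderer, apparently a point light given its light attributes.
# System Timeline values drive the flame's color fade and material curves; a User.bIsSwinging input is exposed.
# Embers spawn as short-lived sprites along a box, rising with curl noise, drag, brief inherited velocity and speed-based stretching.
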NiagaraSystem /BRCosmetics/Weapons/FORT_Melee/Pickaxe_GalaxyLevel/FX/NS_GalaxyLevel_Pickaxe_Idlev2
dependencies {
 usedDataInterfaces [&NiagaraDataInterfaceCurve]
 usedDynamicInputs []
 usedModules [&SystemState, &Timeline]
}
emitters {
 Embers {
  emitterUpdateScript [
   &EmitterState()
   &SpawnBurst_Instantaneous(
    Spawn Count=&RandomRangeInt(Minimum=0, Maximum=5)
    Use Spawn Probability=false
    UseLoopCountLimit=false
   )
   &SpawnRate(SpawnRate=&RandomRangeFloat(Minimum=3, Maximum=12), Use Spawn Probability=false)
  ]
  particleSpawnScript [
   &InitializeParticle(
    Color=<rgba 2 0 1.8333 1>
    Lifetime Max=0.85
    Lifetime Min=0.25
    Lifetime Mode=::NewEnumerator1
    Sprite Size Mode=::NewEnumerator3
    Uniform Sprite Size=2
    UsePositionOffset=false
    Write Material Random=true
   )
   &ShapeLocation(
    Box Midpoint=<0 0.5 0.5>
    Box Size=<35 0 0>
    MaskSpawnGroup=false
    Offset=<-35 0 34>
    Offset Coordinate Space=::Local
    Offset Mode=::NewEnumerator0
    Override Seed=false
    Rotation Mode=::NewEnumerator2
    Shape Primitive=::NewEnumerator2
    Surface Only=false
    UseOwnerScale=true
    Yaw / Pitch / Roll=<0 0.115 0>
   )
   &AddVelocity(Velocity=&RandomRangeVector(Minimum=<-10 -10 25>, Maximum=<10 10 75>))
  ]
  particleUpdateScript [
   &ParticleState()
   &ScaleSpriteSize(Scale Factor=&Vector2DFromFloat(Value=&FloatFromCurve(FloatCurve=DataInterface() {propertyValues {Curve {keys [
          {arriveTangent=9.1531, interpMode=RCIM_Cubic, leaveTangent=9.1531, tangentMode=RCTM_User, time=0, value=0.02}
          {arriveTangent=0.48, interpMode=RCIM_Cubic, leaveTangent=0.48, time=0.18, value=0.99}
          <at 1 0.5 RCIM_Cubic>
         ]}}})))
   &DynamicMaterialParameters(
    Index 0 Param 1=1
    Index 0 Param 2=25
    Index 0 Param 3=1
    Index 1 Param 1=1
    Index 1 Param 2=25
    Index 1 Param 3=3
    Module.Param0WriteEnabled=true
    Module.Param0WriteEnabled1=true
    Module.Param1WriteEnabled=true
    Module.Param1WriteEnabled1=true
    Module.Param2WriteEnabled=true
    Module.Param2WriteEnabled1=true
    Module.Param3WriteEnabled=true
    Module.Param3WriteEnabled1=true
    Write Parameter Index 1=true
   )
   &CurlNoiseForce(
    Noise Frequency=0.0125
    Noise Strength=&FloatFromCurve(FloatCurve=DataInterface() {propertyValues {Curve {keys [<at 0 1>,<at 1 0>]}}}, Scale Curve=250)
    Pan Noise=true
    Pan Noise Field=<0 0 1>
   )
   &AccelerationForce(Acceleration=<0 0 250>)
   &Drag(Drag=5, Use Linear Drag=true, Use Rotational Drag=true)
   &InheritVelocity(
    Inherited Velocity Amount Scale=&Multiply_VectorByFloat(
     Vector=<1 1 1>
     Float=&FloatFromCurve(FloatCurve=DataInterface() {propertyValues {Curve {keys [{arriveTangent=-14.5009, interpMode=RCIM_Cubic, leaveTangent=-14.5009, tangentMode=RCTM_User, time=0, value=1}, <at 0.1 0 RCIM_Cubic>]}}})
    )
    Speed Limit=true
    UseSpeedThreshold=false
   )
   &SolveForcesAndVelocity(Clamp Velocity=false, Limit Acceleration=false)
   &ScaleSpriteSizeBySpeed(Max Scale Factor=<1 4>, Min Scale Factor=<0.5 1>)
   &ScaleColor(
    Scale Alpha=&FloatFromCurve(FloatCurve=DataInterface() {propertyValues {Curve {keys [
         {arriveTangent=11.9433, interpMode=RCIM_Cubic, leaveTangent=11.9433, tangentMode=RCTM_User, time=0, value=0}
         <at 0.2 1 RCIM_Cubic>
         <at 1 0 RCIM_Cubic>
        ]}}})
    Scale RGB=&VectorFromFloat(Value=&FloatFromCurve(FloatCurve=DataInterface() {propertyValues {Curve {keys [<at 0 1>,<at 1 0>]}}}, Scale Curve=10))
    ScaleA=true
    ScaleRGB=true
   )
  ]
  rendererClasses [&NiagaraSpriteRendererProperties]
  renderers [{rendererClass=&NiagaraSpriteRendererProperties}]
 }
 FlameMesh {
  emitterUpdateScript [
   &EmitterState(
    Inactive Response=::NewEnumerator1
    Life Cycle Mode=::NewEnumerator1
    Loop Behavior=::NewEnumerator1
    Loop Duration Mode=::NewEnumerator1
    UseLoopDelay=false
   )
   &SpawnBurst_Instantaneous(Use Spawn Probability=false, UseLoopCountLimit=false)
   /SetVariables(Emitter.Velocity:Vector3f=&NormalizeVector(
     Vector To Normalize=&Multiply_VectorByFloat(Vector=$Engine.Owner.Velocity, Float=-1)
     Safe Normalize=true
     Safe Fallback Threshold=0
     Safe Fallback Vector=<0 0 0.001>
    )) {isSetParameters=true}
  ]
  particleSpawnScript [&InitializeParticle(Color=<rgba 2 0 1.8333 3>, Lifetime=0, Mesh Scale Mode=::NewEnumerator4, UsePositionOffset=false, Write Material Random=true)]
  particleUpdateScript [
   &ParticleState(Kill Particles When Lifetime Has Elapsed=false, Let Infinitely Lived Particles Die When Emitter Deactivates=true)
   &DynamicMaterialParameters(
    Index 0 Param 1=&Waveform(
     "Waveform [1]"=::NewEnumerator9
     "[1] Frequency"=2
     "[1] Compound Frequency"=0.2
     "[1] Amplitude Min/Max"=<0.5 1>
     "[1] Bias"=&RandomRangeFloat(Minimum=0, Maximum=1)
    )
    Index 0 Param 2=&Waveform(
     "Waveform [1]"=::NewEnumerator9
     "[1] Frequency"=1
     "[1] Amplitude Min/Max"=<0 1>
     "[1] Bias"=&RandomRangeFloat(Minimum=0.5, Maximum=0.75)
    )
    Index 0 Param 3=&RandomRangeFloat(Minimum=0.5, Maximum=1)
    Index 0 Param 4=&Waveform(Global Amplitude Scale=0.5, "Waveform [1]"=::NewEnumerator9, "[1] Frequency"=1, "[1] Amplitude Min/Max"=<1 2>, "[1] Bias"=1)
    Index 1 Param 1=&FloatFromCurve(
     FloatCurve=DataInterface() {propertyValues {Curve {keys [{arriveTangent=-2.2801, interpMode=RCIM_Cubic, leaveTangent=-2.2801, tangentMode=RCTM_User, time=0, value=1}, <at 1 0 RCIM_Cubic>]}}}
     CurveIndex=$System.Timeline.NormalizedTime
    )
    Index 1 Param 2=&FloatFromCurve(
     FloatCurve=DataInterface() {propertyValues {Curve {keys [{arriveTangent=2.3106, interpMode=RCIM_Cubic, leaveTangent=2.3106, tangentMode=RCTM_User, time=0, value=0}, <at 1 1 RCIM_Cubic>]}}}
     CurveIndex=$System.Timeline001.NormalizedTime
     Scale Curve=15
    )
    Index 1 Param 3=&RandomRangeFloat(Minimum=0.1, Maximum=0.333)
    Module.Param0WriteEnabled=true
    Module.Param0WriteEnabled1=true
    Module.Param1WriteEnabled=true
    Module.Param1WriteEnabled1=true
    Module.Param2WriteEnabled=true
    Module.Param2WriteEnabled1=true
    Module.Param3WriteEnabled=true
    Module.Param3WriteEnabled1=false
    Write Parameter Index 1=true
   )
   /SetVariables(
    Particles.LightExponent=4
    Particles.LightPosition:NiagaraPosition=&AddVectorToPosition(Position=$Particles.Position, Vector=<-25 0 55>)
    Particles.LightRadius=75
    Particles.PointLightExponent=4
   ) {isSetParameters=true}
   &ScaleColor(
    Scale RGB=&VectorFromFloat(Value=&FloatFromCurve(
      FloatCurve=DataInterface() {propertyValues {Curve {keys [{arriveTangent=2.1152, interpMode=RCIM_Cubic, leaveTangent=2.1152, tangentMode=RCTM_User, time=0, value=0}, <at 1 1 RCIM_Cubic>]}}}
      CurveIndex=$System.Timeline.NormalizedTime
     ))
    ScaleA=true
    ScaleRGB=true
   )
  ]
  rendererClasses [&NiagaraMeshRendererProperties, &NiagaraComponentRendererProperties]
  renderers [
   {rendererClass=&NiagaraMeshRendererProperties, rendererIndex=0}
   {rendererClass=&NiagaraComponentRendererProperties, rendererIndex=1}
  ]
 }
}
summary {
 systemName=NS_GalaxyLevel_Pickaxe_Idlev2
 emitters [
  {emitterName=FlameMesh, rendererClasses [&NiagaraMeshRendererProperties, &NiagaraComponentRendererProperties]}
  {emitterName=Embers, rendererClasses [&NiagaraSpriteRendererProperties]}
 ]
 userVariables [User.bIsSwinging {type=NiagaraBool}]
}
